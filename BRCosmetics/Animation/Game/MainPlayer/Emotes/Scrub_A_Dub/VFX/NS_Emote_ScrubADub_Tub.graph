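# Niagara VFX for the Scrub-A-Dub emote's tub: splashes, foam, bubbles and brush effects.
# Bursts make splash mesh, foam and drops, while System.SplashN/E/S/W gate quadrant spawns through spawn groups.
# Bubbles slide down the tub along a curve-driven ring radius, wobbling and popping through scale and material curves.
# Brush emitters follow a User.SkeletalMesh bone or socket, with User.FoamSpawn gating foam and bubbles and timeline time driving drip rate.
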NiagaraSystem /BRCosmetics/Animation/Game/MainPlayer/Emotes/Scrub_A_Dub/VFX/NS_Emote_ScrubADub_Tub
dependencies {
 usedDataInterfaces [
  &NiagaraDataInterfaceCurve
  &NiagaraDataInterfaceVector2DCurve
  &NiagaraDataInterfaceVectorCurve
 ]
 usedDynamicInputs [&Multiply_Vector2D_ByFloat]
 usedModules [
  &self:SystemSpawnScript.SystemScriptSource.SystemScriptGraph.NiagaraNodeAssignment_0.SetVariables_#1
  &self:SystemSpawnScript.SystemScriptSource.SystemScriptGraph.NiagaraNodeAssignment_2.SetVariables_#2
  &SystemState
  &SampleSkeletalMesh
  &self:SystemSpawnScript.SystemScriptSource.SystemScriptGraph.NiagaraNodeAssignment_5.SetVariables_#3
  &self:SystemSpawnScript.SystemScriptSource.SystemScriptGraph.NiagaraNodeAssignment_1.SetVariables_#4
  &Timeline
 ]
}
emitters {
 Bubbles_Brush {
  emitterUpdateScript [
   &EmitterState(
    Inactive Response=::NewEnumerator1
    Life Cycle Mode=::NewEnumerator1
    Loop Behavior=::NewEnumerator1
    Loop Duration Mode=::NewEnumerator1
    UseLoopDelay=false
   )
   &SpawnBurst_Instantaneous(Spawn Count=11, Spawn Group=1, Use Spawn Probability=false, UseLoopCountLimit=false)
   &SpawnRate(Spawn Probability=$User.FoamSpawn, SpawnRate=20, Use Spawn Probability=true)
  ]
  particleSpawnScript [
   /SetVariables(Particles.SpawnGroup=$Emitter.SpawnGroup) {isSetParameters=true}
   &ShapeLocation(
    Apply To Particle Position=::NewEnumerator1
    Cylinder Height=0.5
    Cylinder Radius=6
    MaskSpawnGroup=true
    Non Uniform Scale=<0.75 1 1>
    Override Seed=false
    Rotation Axis=<0 -1 0>
    Rotation Coordinate Space=::World
    Rotation Mode=::NewEnumerator1
    Shape Origin=&VectorFromFloat(Value=0)
    Shape Primitive=::NewEnumerator1
    Surface Only=false
    UseOwnerScale=true
   )
   &ShapeLocation(
    Apply To Particle Position=::NewEnumerator1
    MaskSpawnGroup=true
    Non Uniform Scale=<0.75 1 1>
    Ring / Disc Distribution Mode:ENiagaraSphereDistributionMode=::NewEnumerator3
    Ring Radius=6
    Rotation Axis=<0 -1 0>
    Rotation Coordinate Space=::World
    Rotation Mode=::NewEnumerator1
    Shape Origin=&MakeVector(X=0, Y=&RandomRangeFloat(Minimum=-0.25, Maximum=-0.75), Z=0)
    Shape Primitive=::NewEnumerator4
    Spawn Group Mask=1
    UseOwnerScale=true
   ) {module=ShapeLocation001}
   &SkeletalMeshLocation(
    Bone / Socket Index=6
    Bone Sampling Mode=::NewEnumerator4
    MaskSpawnGroup=false
    Orientation Sampling=::NewEnumerator1
    Sampled Position Offset=&MaskVectorBySpawnGroup(Masked Vector=$Particles.Initial.ShapeLocation.ShapePosition, Passthrough Vector=$Particles.Initial.ShapeLocation001.ShapePosition)
    Skeletal Mesh=$User.SkeletalMesh
   )
   &InitializeParticle(
    Lifetime=3
    Mesh Scale Mode=::NewEnumerator3
    Mesh Uniform Scale=&MaskFloatBySpawnGroup(
     Masked Float=&RandomRangeFloat(Minimum=1, Maximum=2)
     Passthrough Float=&RandomRangeFloat(Minimum=2, Maximum=3)
    )
    UsePositionOffset=false
    Write Material Random=true
   )
  ]
  particleUpdateScript [
   &ParticleState(
    Kill Particles When Lifetime Has Elapsed=false
    Loop Particles Lifetime:NiagaraBool=/Niagara/DynamicInputs/Helpers/MaskBoolBySpawnGroup(Masked Bool:NiagaraBool=false, Passthrough Bool:NiagaraBool=true, Spawn Group Mask:NiagaraInt32=0)
   )
   &KillParticles(Kill Particles=&SetBoolByFloatComparison(A=$Particles.NormalizedAge, B=1))
   &SkeletalMeshLocation(
    Bone / Socket Index=6
    Bone Sampling Mode=::NewEnumerator4
    MaskSpawnGroup=false
    Orientation Sampling=::NewEnumerator1
    Sampled Position Offset=&MaskVectorBySpawnGroup(Masked Vector=$Particles.Initial.ShapeLocation.ShapePosition, Passthrough Vector=$Particles.Initial.ShapeLocation001.ShapePosition)
    Skeletal Mesh=$User.SkeletalMesh
   ) {module=SkeletalMeshLocation001}
   /SetVariables(Particles.BubbleTimeRandom:NiagaraFloat=&Add_Float(A=$Emitter.Age, B=&RandomRangeFloat(Minimum=0, Maximum=1))) {isSetParameters=true}
   &DynamicMaterialParameters(
    Index 0 Param 1=&FloatFromCurve(
     FloatCurve=DataInterface() {propertyValues {Curve {keys [{tangentMode=RCTM_User, time=0.95, value=0}, {tangentMode=RCTM_User, time=1, value=1}]}}}
     Scale Curve=&MaskFloatBySpawnGroup(Masked Float=1, Passthrough Float=0)
    )
    Module.Param0WriteEnabled=true
    Module.Param1WriteEnabled=false
    Module.Param2WriteEnabled=false
    Module.Param3WriteEnabled=false
   )
   &ScaleMeshSize(Scale Factor=&MakeVector(
     X=&Sine(Normalized Angle=$Particles.BubbleTimeRandom, Period=0.33, Scale=0.05, Bias=1)
     Y=&Sine(Normalized Angle=$Particles.BubbleTimeRandom, Period=0.33, Scale=0.05, Bias=1)
     Z=&Cosine(Normalized Angle=$Particles.BubbleTimeRandom, Period=0.33, Scale=0.05, Bias=1)
    )) {module=ScaleMeshSize002}
   &ScaleMeshSize(Scale Factor=&MaskVectorBySpawnGroup(
     Masked Vector=&VectorFromFloat(Value=&FloatFromCurve(FloatCurve=DataInterface() {propertyValues {Curve {keys [{arriveTangent=21.9281, interpMode=RCIM_Cubic, leaveTangent=21.9281, tangentMode=RCTM_User, time=0.95, value=1}, <at 1 1.4 RCIM_Cubic>]}}}))
     Passthrough Vector=<1 1 1>
    )) {module=ScaleMeshSize001}
  ]
  rendererClasses [&NiagaraMeshRendererProperties]
  renderers [{rendererClass=&NiagaraMeshRendererProperties, rendererIndex=0}]
 }
 Bubbles_Flowing {
  emitterUpdateScript [
   &EmitterState(
    Inactive Response=::NewEnumerator1
    Life Cycle Mode=::NewEnumerator1
    Loop Behavior=::NewEnumerator1
    Loop Duration Mode=::NewEnumerator1
    UseLoopDelay=false
   )
   &SpawnBurst_Instantaneous(
    Spawn Count=&RandomRangeInt(Minimum=30, Maximum=45)
    Use Spawn Probability=false
    UseLoopCountLimit=true
   )
   &SpawnRate(Spawn Probability=$System.SplashN, SpawnRate=60, Use Spawn Probability=true) {module=SpawnRate001}
   &SpawnRate(Spawn Probability=$System.SplashE, SpawnGroup=1, SpawnRate=60, Use Spawn Probability=true) {module=SpawnRate002}
   &SpawnRate(Spawn Probability=$System.SplashS, SpawnGroup=2, SpawnRate=60, Use Spawn Probability=true) {module=SpawnRate003}
   &SpawnRate(Spawn Probability=$System.SplashW, SpawnGroup=3, SpawnRate=60, Use Spawn Probability=true) {module=SpawnRate004}
  ]
  particleSpawnScript [
   &InitializeParticle(
    Lifetime Max=5
    Lifetime Min=4
    Lifetime Mode=::NewEnumerator1
    Mesh Scale Mode=::NewEnumerator1
    Mesh Uniform Scale Max=3
    Mesh Uniform Scale Min=0.8
    Position Offset=<0 0 80>
    UsePositionOffset=false
    Write Material Random=true
   )
   &InitialMeshOrientation(Mesh Axis To Orient=<0 0 1>, Mesh Orientation Mode=::NewEnumerator2, Use Rotation Vector=false)
   &ShapeLocation(
    MaskSpawnGroup=true
    Offset=<0 0 77>
    Offset Mode=::NewEnumerator0
    Override Seed=false
    Ring Radius=30
    Rotation Angle:NiagaraFloat=&Add_Float(A=0.625, B=&Multiply_FloatByInt(Float=0.25, Integer=$Emitter.SpawnGroup))
    Rotation Axis=<0 0 1>
    Rotation Mode=::NewEnumerator1
    Shape Origin=$Engine.Emitter.SimulationPosition
    Shape Primitive=::NewEnumerator4
    UseOwnerScale=true
   ) {module=ShapeLocation001}
  ]
  particleUpdateScript [
   /SetVariables(Particles.DeltaTime:NiagaraFloat=&Waveform(
     Global Amplitude Scale=&RemapRange(
      Input Value To Remap=&MakeFloatFromVector(Vector=$Particles.Initial.Scale, Channel=::NewEnumerator0)
      Input Min=0.8
      Input Max=3
      Output Min=0.5
      Output Max=1
     )
     "Waveform [1]"=::NewEnumerator13
     "[1] Input Position"=&RandomRangeFloat(Minimum=0, Maximum=10)
     "[1] Frequency"=2
     "[1] Amplitude Min/Max"=<0.02 1>
    )) {isSetParameters=true}
   &ParticleState(DeltaTime:NiagaraFloat=&Multiply_Float(A=$Engine.DeltaTime, B=$Particles.DeltaTime))
   /SetVariables(Particles.BubbleTimeRandom:NiagaraFloat=&Add_Float(A=$Emitter.Age, B=&RandomRangeFloat(Minimum=0, Maximum=1))) {isSetParameters=true}
   &ShapeLocation(
    MaskSpawnGroup=true
    Offset=&Lerp_Vector(
     A=<0 0 77>
     B=<1 1 0>
     Alpha=&FloatFromCurve(FloatCurve=DataInterface() {propertyValues {Curve {keys [
          {arriveTangent=-12.9569, interpMode=RCIM_Cubic, time=0, value=0}
          <at 0.0103 -0.05 RCIM_Cubic>
          {arriveTangent=1.1801, interpMode=RCIM_Cubic, leaveTangent=1.1801, time=0.02, value=0}
          {interpMode=RCIM_Cubic, leaveTangent=1.2196, time=0.9, value=1}
         ]}}})
    )
    Offset Mode=::NewEnumerator0
    Override Seed=false
    Ring / Disc Distribution Mode:ENiagaraSphereDistributionMode=::NewEnumerator1
    Ring Radius=&Lerp_Float(
     A=30
     B=38
     Alpha=&FloatFromCurve(FloatCurve=DataInterface() {propertyValues {Curve {keys [
          {arriveTangent=60.6621, time=0, value=0}
          <at 0.02 1>
          <at 0.12 1>
          <at 0.14 1.15>
          <at 0.19 1.15>
          <at 0.21 1>
          <at 0.35 1>
          <at 0.37 1.15>
          <at 0.42 1.15>
          <at 0.44 1>
          <at 0.57 1>
          <at 0.59 1.15>
          <at 0.64 1.15>
          <at 0.66 1>
          {arriveTangent=3.0387, leaveTangent=3.0387, time=1, value=1}
         ]}}})
    )
    Shape Primitive=::NewEnumerator4
    U Position=&RandomRangeFloat(
     Minimum=&RandomRangeFloat(Minimum=1, Maximum=0)
     Maximum=&RandomRangeFloat(Minimum=0, Maximum=1)
    )
    UseOwnerScale=true
   )
   &UpdateMeshOrientation(
    Facing Direction=&NormalizeVector(
     Vector To Normalize=&Subtract_Position(A=$Particles.Previous.Position, B=$Particles.Position)
     Safe Normalize=false
     Safe Fallback Threshold=0
     Safe Fallback Vector=<0 0 0.001>
    )
    Facing Interpolation=true
    Facing Interpolation Rate=6
    Orientation Method=::NewEnumerator1
    Override Quat=false
    Side Direction:Vector3f=$Particles.ShapeLocation.ShapeTangent
    Stabilization Mode:ENiagara_OrientVectorCount=::NewEnumerator2
    TransformFacing=false
    TransformSide=false
   )
   &ScaleMeshSize(Scale Factor=&Lerp_Vector(
     A=<1 1 0.9>
     B=<0.9 0.9 1.2>
     Alpha=&Multiply_Float(
      A=$Particles.DeltaTime
      B=&FloatFromCurve(FloatCurve=DataInterface() {propertyValues {Curve {keys [<at 0.02 1 RCIM_Constant>,<at 0.9 0>]}}})
     )
    ))
   &ScaleMeshSize(Scale Factor=&MakeVector(
     X=&Sine(Normalized Angle=$Particles.BubbleTimeRandom, Period=0.33, Scale=0.05, Bias=1)
     Y=&Sine(Normalized Angle=$Particles.BubbleTimeRandom, Period=0.33, Scale=0.05, Bias=1)
     Z=&Cosine(Normalized Angle=$Particles.BubbleTimeRandom, Period=0.33, Scale=0.05, Bias=1)
    )) {module=ScaleMeshSize002}
   &ScaleMeshSize(Scale Factor=&VectorFromFloat(Value=&FloatFromCurve(FloatCurve=DataInterface() {propertyValues {Curve {keys [{arriveTangent=25.9644, interpMode=RCIM_Cubic, leaveTangent=25.9644, tangentMode=RCTM_User, time=0.98, value=1}, <at 1 1.4 RCIM_Cubic>]}}}))) {module=ScaleMeshSize001}
   &DynamicMaterialParameters(
    Index 0 Param 1=&FloatFromCurve(FloatCurve=DataInterface() {propertyValues {Curve {keys [{tangentMode=RCTM_User, time=0.99, value=0}, {tangentMode=RCTM_User, time=1, value=1}]}}})
    Module.Param0WriteEnabled=true
    Module.Param1WriteEnabled=false
    Module.Param2WriteEnabled=false
    Module.Param3WriteEnabled=false
   )
  ]
  rendererClasses [&NiagaraMeshRendererProperties]
  renderers [
   {rendererClass=&NiagaraMeshRendererProperties, rendererIndex=0}
   {rendererClass=&NiagaraMeshRendererProperties, rendererIndex=1}
  ]
 }
 FN_Water_Drops_ {
  emitterUpdateScript [
   &EmitterState(Life Cycle Mode=::NewEnumerator1, Loop Behavior=::NewEnumerator1, Loop Delay=$System.SplashDelay, Loop Duration=1, UseLoopDelay=true)
   &SpawnBurst_Instantaneous(Spawn Count=120, Use Spawn Probability=false, UseLoopCountLimit=false)
   &SpawnBurst_Instantaneous(Spawn Count=30, Spawn Group=1, Use Spawn Probability=false, UseLoopCountLimit=false) {module=SpawnBurst_Instantaneous001}
  ]
  particleSpawnScript [
   /SetVariables(Particles.VisibilityTag=&SpawnGroup()) {isSetParameters=true}
   &InitialMeshOrientation(Use Rotation Vector=false)
   &InitializeParticle(
    Color=<rgba 1 1 1 0.4>
    Lifetime=&MaskFloatBySpawnGroup(
     Masked Float=&RandomRangeFloat(Minimum=0.25, Maximum=1)
     Passthrough Float=&RandomRangeFloat(Minimum=1, Maximum=3)
    )
    Mesh Scale Mode=::NewEnumerator1
    Mesh Uniform Scale Max=3
    Mesh Uniform Scale Min=1.5
    Position Offset=<0 0 72>
    Sprite Size=&Vector2DFromCurve(
     Vector2Curve=DataInterface() {propertyValues {
       XCurve {keys [<at -0.0011 2.5 RCIM_Cubic>,<at 0.9988 1 RCIM_Cubic>]}
       YCurve {keys [
         {arriveTangent=-30.8979, interpMode=RCIM_Cubic, leaveTangent=-30.8979, tangentMode=RCTM_User, time=0, value=9}
         {arriveTangent=-6.5075, interpMode=RCIM_Cubic, leaveTangent=-6.5075, time=0.2382, value=5}
         <at 0.9988 2.5 RCIM_Cubic>
        ]}
      }}
     CurveIndex=&RandomRangeFloat(Minimum=0, Maximum=1)
     ScaleCurve=<1 1>
    )
    Sprite Size Mode=::NewEnumerator4
    UsePositionOffset=true
    Write Material Random=true
   )
   &DynamicMaterialParameters(
    Index 0 Param 1=&UniformRangedFloat(Minimum=1, Maximum=3)
    Index 1 Param 1=&FloatFromCurve(FloatCurve=DataInterface() {propertyValues {Curve {keys [{tangentMode=RCTM_User, time=0.99, value=0}, {tangentMode=RCTM_User, time=1, value=1}]}}})
    Module.Param0WriteEnabled=true
    Module.Param0WriteEnabled1=false
    Module.Param1WriteEnabled=false
    Module.Param1WriteEnabled1=true
    Module.Param2WriteEnabled=false
    Module.Param2WriteEnabled1=true
    Module.Param3WriteEnabled=false
    Module.Param3WriteEnabled1=true
    Write Parameter Index 1=true
   )
   &ShapeLocation(
    Disc Coverage=0.3
    MaskSpawnGroup=false
    Offset Mode=::NewEnumerator0
    Override Seed=false
    Ring Radius=30
    Shape Primitive=::NewEnumerator4
    UseOwnerScale=true
   )
   &AddVelocity(
    Module.Use Falloff Distance=false
    Origin Offset=&AddVectorToPosition(
     Position=&MaskVectorBySpawnGroup(Masked Vector=<0 0 70>, Passthrough Vector=<0 0 60>)
     Vector=&RandomRangeVector(Minimum=<0 0 0>, Maximum=<0 0 -140>)
    )
    Override Seed=false
    Velocity Mode=::NewEnumerator1
    Velocity Speed=&MaskFloatBySpawnGroup(
     Masked Float=&RandomRangeFloat(Minimum=150, Maximum=500)
     Passthrough Float=&RandomRangeFloat(Minimum=150, Maximum=500)
    )
   )
  ]
  particleUpdateScript [
   &ParticleState()
   &CurlNoiseForce(
    Noise Frequency=60
    Noise Quality / Cost=::NewEnumerator0
    Noise Strength=&MaskFloatBySpawnGroup(Masked Float=0, Passthrough Float=150)
    Pan Noise=false
   )
   &Drag(
    Drag=&MaskFloatBySpawnGroup(Masked Float=0.25, Passthrough Float=&RandomRangeFloat(Minimum=3, Maximum=5))
    Use Linear Drag=true
    Use Rotational Drag=true
   )
   &GravityForce(Gravity=&MaskVectorBySpawnGroup(Masked Vector=<0 0 -980>, Passthrough Vector=&RandomRangeVector(Minimum=<0 0 -25>, Maximum=<0 0 -100>)))
   &SolveForcesAndVelocity(Clamp Velocity=false, Limit Acceleration=false)
   &DynamicMaterialParameters(
    Index 1 Param 1=&FloatFromCurve(FloatCurve=DataInterface() {propertyValues {Curve {keys [{tangentMode=RCTM_User, time=0.95, value=0}, {tangentMode=RCTM_User, time=1, value=1}]}}})
    Module.Param0WriteEnabled1=true
    Module.Param1WriteEnabled1=true
    Module.Param2WriteEnabled1=true
    Module.Param3WriteEnabled1=true
    Write Parameter Index 0=false
    Write Parameter Index 1=true
   ) {module=DynamicMaterialParameters001}
   &ScaleColor(
    Scale Alpha=&FloatFromCurve(FloatCurve=DataInterface() {propertyValues {Curve {keys [
         {arriveTangent=28.4939, interpMode=RCIM_Cubic, leaveTangent=28.4939, tangentMode=RCTM_User, time=0, value=0}
         {arriveTangent=0.002, interpMode=RCIM_Cubic, leaveTangent=0.002, time=0.1183, value=1.0113}
         {arriveTangent=-3.6864, interpMode=RCIM_Cubic, leaveTangent=-3.6864, tangentMode=RCTM_User, time=1.0011, value=0.002}
        ]}}})
    ScaleA=true
    ScaleRGB=true
   )
   /SetVariables(Particles.BubbleTimeRandom:NiagaraFloat=&Add_Float(A=$Emitter.Age, B=&RandomRangeFloat(Minimum=0, Maximum=1))) {isSetParameters=true}
   &ScaleMeshSize(Scale Factor=&VectorFromFloat(Value=&FloatFromCurve(FloatCurve=DataInterface() {propertyValues {Curve {keys [{arriveTangent=10.0536, interpMode=RCIM_Cubic, leaveTangent=10.0536, tangentMode=RCTM_User, time=0.9, value=1}, <at 1 1.4 RCIM_Cubic>]}}}))) {module=ScaleMeshSize001}
   &ScaleMeshSize(Scale Factor=&MakeVector(
     X=&Sine(Normalized Angle=$Particles.BubbleTimeRandom, Period=0.4, Scale=0.05, Bias=1)
     Y=&Sine(Normalized Angle=$Particles.BubbleTimeRandom, Period=0.4, Scale=0.05, Bias=1)
     Z=&Cosine(Normalized Angle=$Particles.BubbleTimeRandom, Period=0.4, Scale=0.05, Bias=1)
    )) {module=ScaleMeshSize002}
  ]
  rendererClasses [&NiagaraSpriteRendererProperties, &NiagaraMeshRendererProperties]
  renderers [
   {rendererClass=&NiagaraSpriteRendererProperties}
   {rendererClass=&NiagaraMeshRendererProperties, rendererIndex=1}
  ]
 }
 FN_Water_Drops_001 {
  emitterUpdateScript [
   &EmitterState()
   &SpawnRate(Spawn Probability=$System.SplashN, SpawnRate=30, Use Spawn Probability=true)
   &SpawnRate(Spawn Probability=$System.SplashE, SpawnGroup=1, SpawnRate=30, Use Spawn Probability=true) {module=SpawnRate001}
   &SpawnRate(Spawn Probability=$System.SplashS, SpawnGroup=2, SpawnRate=30, Use Spawn Probability=true) {module=SpawnRate003}
   &SpawnRate(Spawn Probability=$System.SplashW, SpawnGroup=3, SpawnRate=30, Use Spawn Probability=true) {module=SpawnRate002}
  ]
  particleSpawnScript [
   &InitializeParticle(
    Color Minimum=<rgba 1 1 1 0.5>
    Color Mode=::NewEnumerator2
    Lifetime Max=1
    Lifetime Min=0.25
    Lifetime Mode=::NewEnumerator1
    Mesh Scale Mode=::NewEnumerator1
    Mesh Uniform Scale Max=3
    Mesh Uniform Scale Min=1.5
    Position Offset=<0 0 72>
    Sprite Size=&Vector2DFromCurve(
     Vector2Curve=DataInterface() {propertyValues {
       XCurve {keys [<at -0.0011 2.5 RCIM_Cubic>,<at 0.9988 1 RCIM_Cubic>]}
       YCurve {keys [
         {arriveTangent=-30.8979, interpMode=RCIM_Cubic, leaveTangent=-30.8979, tangentMode=RCTM_User, time=0, value=9}
         {arriveTangent=-6.5075, interpMode=RCIM_Cubic, leaveTangent=-6.5075, time=0.2382, value=5}
         <at 0.9988 2.5 RCIM_Cubic>
        ]}
      }}
     CurveIndex=&RandomRangeFloat(Minimum=0, Maximum=1)
     ScaleCurve=<1 1>
    )
    Sprite Size Mode=::NewEnumerator4
    UsePositionOffset=true
    Write Material Random=true
   )
   &ShapeLocation(
    Disc Coverage=0.3
    MaskSpawnGroup=false
    Offset Mode=::NewEnumerator0
    Override Seed=false
    Ring Radius=27
    Rotation Angle:NiagaraFloat=&Add_Float(A=0.625, B=&Multiply_FloatByInt(Float=0.25, Integer=$Emitter.SpawnGroup))
    Rotation Axis=<0 0 1>
    Rotation Mode=::NewEnumerator1
    Shape Primitive=::NewEnumerator4
    U Distribution=0.75
    UseOwnerScale=true
   )
   &DynamicMaterialParameters(
    Index 0 Param 1=&UniformRangedFloat(Minimum=1, Maximum=3)
    Module.Param0WriteEnabled=true
    Module.Param1WriteEnabled=false
    Module.Param2WriteEnabled=false
    Module.Param3WriteEnabled=false
   )
   &self:SplashVelocityScale() {module=SplashVelocityScale}
   &AddVelocity(
    Module.Use Falloff Distance=false
    Origin Offset=&RandomRangeVector(Minimum=<0 0 25>, Maximum=<0 0 -25>)
    Override Seed=false
    Velocity Mode=::NewEnumerator1
    Velocity Speed=&Multiply_Float(A=&RandomRangeFloat(Minimum=200, Maximum=260), B=$Particles.SplashVelocityScale)
   )
  ]
  particleUpdateScript [
   &ParticleState()
   &Drag(Use Linear Drag=true, Use Rotational Drag=false)
   &GravityForce()
   &SolveForcesAndVelocity(Clamp Velocity=false, Limit Acceleration=false)
   &ScaleColor(
    Scale Alpha=&FloatFromCurve(FloatCurve=DataInterface() {propertyValues {Curve {keys [
         {arriveTangent=28.4939, interpMode=RCIM_Cubic, leaveTangent=28.4939, tangentMode=RCTM_User, time=0, value=0}
         {arriveTangent=0.002, interpMode=RCIM_Cubic, leaveTangent=0.002, time=0.1183, value=1.0113}
         {arriveTangent=-3.6864, interpMode=RCIM_Cubic, leaveTangent=-3.6864, tangentMode=RCTM_User, time=1.0011, value=0.002}
        ]}}})
    ScaleA=true
    ScaleRGB=true
   )
  ]
  rendererClasses [&NiagaraSpriteRendererProperties]
  renderers [{rendererClass=&NiagaraSpriteRendererProperties}]
 }
 Foam_Brush {
  emitterUpdateScript [
   &EmitterState()
   &SpawnRate(Spawn Probability=$User.FoamSpawn, SpawnRate=15, Use Spawn Probability=true)
  ]
  particleSpawnScript [
   &InitializeParticle(
    Lifetime Max=0.5
    Lifetime Min=0.4
    Lifetime Mode=::NewEnumerator1
    Mesh Scale Mode=::NewEnumerator3
    Position Offset=<0 0 75>
    Sprite Rotation Mode=::NewEnumerator2
    Sprite Size Mode=::NewEnumerator1
    Uniform Sprite Size Max=25
    Uniform Sprite Size Min=20
    UsePositionOffset=true
    Write Material Random=true
   ) {module=InitializeParticle001}
   &ShapeLocation(
    Apply To Particle Position=::NewEnumerator1
    MaskSpawnGroup=false
    Non Uniform Scale=<0.75 1 1>
    Override Seed=false
    Ring Radius=2
    Rotation Axis=<0 -1 0>
    Rotation Coordinate Space=::World
    Rotation Mode=::NewEnumerator1
    Shape Origin=&MakeVector(X=0, Y=2, Z=0)
    Shape Primitive=::NewEnumerator4
    UseOwnerScale=true
   )
   &SkeletalMeshLocation(
    Bone / Socket Index=6
    Bone Sampling Mode=::NewEnumerator4
    "Inherit Velocity (Scale)"=0.4
    MaskSpawnGroup=false
    Orientation Sampling=::NewEnumerator1
    Sampled Position Offset=$Particles.Initial.ShapeLocation.ShapePosition
    Skeletal Mesh=$User.SkeletalMesh
    Velocity Sampling=::NewEnumerator0
    VelocitySpeedLimit=false
   ) {module=SkeletalMeshLocation001}
   /Niagara/Modules/Update/SubUV/SubUVAnimation(SubUV Animation Mode=::NewEnumerator1)
  ]
  particleUpdateScript [
   &ParticleState()
   &InheritVelocity(
    Module.Limit Speed=true
    Velocity Limit=10000
    Velocity Scale=&Multiply_VectorByFloat(Vector=<-0.03 -0.03 -0.03>, Float=0.5)
    Velocity Source=&TransformVector(Vector=$Engine.Owner.Velocity, Source Space=::World, Destination Space=::Local)
   ) {enabled=false}
   &SpriteRotationRate(Rotation Rate=&RandomRangeFloat(Minimum=-90, Maximum=90)) {enabled=false}
   &Drag(Use Linear Drag=true, Use Rotational Drag=false)
   &GravityForce(Gravity=<0 0 -300>)
   &SolveForcesAndVelocity(Clamp Velocity=false, Limit Acceleration=false)
   &ScaleSpriteSize(Scale Factor=&Vector2DFromFloat(Value=&FloatFromCurve(FloatCurve=DataInterface() {propertyValues {Curve {keys [
          {arriveTangent=0.4856, interpMode=RCIM_Cubic, leaveTangent=0.4856, tangentMode=RCTM_User, time=0, value=0.6}
          {interpMode=RCIM_Cubic, tangentMode=RCTM_User, time=1, value=1}
         ]}}})))
   &ScaleColor(
    Scale Alpha=&FloatFromCurve(FloatCurve=DataInterface() {propertyValues {Curve {keys [{interpMode=RCIM_Cubic, tangentMode=RCTM_User, time=0, value=1}, {interpMode=RCIM_Cubic, tangentMode=RCTM_User, time=0.33, value=0.4}]}}})
    Scale RGB=&Lerp_Vector(
     A=<0 0.75 1>
     B=<2 2 2>
     Alpha=&FloatFromCurve(FloatCurve=DataInterface() {propertyValues {Curve {keys [
          {arriveTangent=3.3971, interpMode=RCIM_Cubic, leaveTangent=3.3971, tangentMode=RCTM_User, time=0, value=0}
          {interpMode=RCIM_Cubic, tangentMode=RCTM_User, time=0.45, value=1}
         ]}}})
    )
    ScaleA=false
    ScaleRGB=true
   )
   &DynamicMaterialParameters(
    Index 0 Param 1=&RandomRangeFloat(Minimum=0, Maximum=1)
    Index 0 Param 3=&FloatFromCurve(FloatCurve=DataInterface() {propertyValues {Curve {keys [{interpMode=RCIM_Cubic, tangentMode=RCTM_User, time=0, value=0}, {interpMode=RCIM_Cubic, tangentMode=RCTM_User, time=1, value=1}]}}})
    Index 0 Param 4=&FloatFromCurve(FloatCurve=DataInterface() {propertyValues {Curve {keys [<at 0 0 RCIM_Cubic>, {arriveTangent=1.5276, interpMode=RCIM_Cubic, leaveTangent=1.5276, tangentMode=RCTM_User, time=1, value=1}]}}})
    Module.Param0WriteEnabled=true
    Module.Param1WriteEnabled=false
    Module.Param2WriteEnabled=true
    Module.Param3WriteEnabled=true
   )
  ]
  rendererClasses [&NiagaraSpriteRendererProperties]
  renderers [{rendererClass=&NiagaraSpriteRendererProperties}]
 }
 Splash_Foam {
  emitterUpdateScript [
   &EmitterState(Life Cycle Mode=::NewEnumerator1, Loop Behavior=::NewEnumerator1, Loop Delay=$System.SplashDelay, Loop Duration=1, UseLoopDelay=true)
   &SpawnBurst_Instantaneous(Spawn Count=13, Use Spawn Probability=false, UseLoopCountLimit=false) {module=SpawnBurst_Instantaneous001}
  ]
  particleSpawnScript [
   &InitializeParticle(
    Lifetime Max=1
    Lifetime Min=0.75
    Lifetime Mode=::NewEnumerator1
    Mesh Scale Mode=::NewEnumerator3
    Position Offset=<0 0 75>
    Sprite Rotation Mode=::NewEnumerator2
    Sprite Size Mode=::NewEnumerator1
    Uniform Sprite Size Max=100
    Uniform Sprite Size Min=75
    UsePositionOffset=true
    Write Material Random=true
   ) {module=InitializeParticle001}
   &ShapeLocation(
    Disc Coverage=0.3
    MaskSpawnGroup=false
    Offset Mode=::NewEnumerator0
    Override Seed=false
    Ring Radius=20
    Shape Primitive=::NewEnumerator4
    UseOwnerScale=true
   )
   &SubUVAnimation(SubUV Animation Mode=::NewEnumerator1)
   &AddVelocity(
    Module.Use Falloff Distance=false
    Override Seed=false
    Velocity Mode=::NewEnumerator1
    Velocity Speed=&RandomRangeFloat(Minimum=150, Maximum=350)
   ) {module=AddVelocity001}
  ]
  particleUpdateScript [
   &ParticleState()
   &InheritVelocity(
    Module.Limit Speed=true
    Velocity Limit=10000
    Velocity Scale=&Multiply_VectorByFloat(Vector=<-0.03 -0.03 -0.03>, Float=0.5)
    Velocity Source=&TransformVector(Vector=$Engine.Owner.Velocity, Source Space=::World, Destination Space=::Local)
   )
   &SpriteRotationRate(Rotation Rate=&RandomRangeFloat(Minimum=-90, Maximum=90))
   &Drag(Drag=0.25, Use Linear Drag=true, Use Rotational Drag=false)
   &GravityForce(Gravity=<0 0 -900>)
   &SolveForcesAndVelocity(Clamp Velocity=false, Limit Acceleration=false)
   &ScaleSpriteSize(Scale Factor=&Vector2DFromFloat(Value=&FloatFromCurve(
      FloatCurve=DataInterface() {propertyValues {Curve {keys [
          {arriveTangent=1.6115, interpMode=RCIM_Cubic, leaveTangent=1.6115, tangentMode=RCTM_User, time=0, value=0.33}
          {interpMode=RCIM_Cubic, tangentMode=RCTM_User, time=1, value=1}
         ]}}}
      Scale Curve=0.65
     )))
   &ScaleColor(
    Scale Alpha=&FloatFromCurve(FloatCurve=DataInterface() {propertyValues {Curve {keys [{interpMode=RCIM_Cubic, tangentMode=RCTM_User, time=0, value=1}, {interpMode=RCIM_Cubic, tangentMode=RCTM_User, time=0.33, value=0.4}]}}})
    Scale RGB=&Lerp_Vector(
     A=<0 0.75 1>
     B=<2 2 2>
     Alpha=&FloatFromCurve(FloatCurve=DataInterface() {propertyValues {Curve {keys [{interpMode=RCIM_Cubic, tangentMode=RCTM_User, time=0, value=0}, {interpMode=RCIM_Cubic, tangentMode=RCTM_User, time=0.45, value=1}]}}})
    )
    ScaleA=false
    ScaleRGB=true
   )
   &DynamicMaterialParameters(
    Index 0 Param 1=&RandomRangeFloat(Minimum=0, Maximum=1)
    Index 0 Param 3=&FloatFromCurve(FloatCurve=DataInterface() {propertyValues {Curve {keys [{interpMode=RCIM_Cubic, tangentMode=RCTM_User, time=0, value=0}, {interpMode=RCIM_Cubic, tangentMode=RCTM_User, time=1, value=1}]}}})
    Index 0 Param 4=&FloatFromCurve(FloatCurve=DataInterface() {propertyValues {Curve {keys [<at 0 0 RCIM_Cubic>, {arriveTangent=1.5276, interpMode=RCIM_Cubic, leaveTangent=1.5276, tangentMode=RCTM_User, time=1, value=1}]}}})
    Module.Param0WriteEnabled=true
    Module.Param1WriteEnabled=false
    Module.Param2WriteEnabled=true
    Module.Param3WriteEnabled=true
   )
  ]
  rendererClasses [&NiagaraSpriteRendererProperties]
  renderers [{rendererClass=&NiagaraSpriteRendererProperties}]
 }
 Splash_Foam_Dynamic {
  emitterUpdateScript [
   &EmitterState()
   &SpawnRate(Spawn Probability=$System.SplashN, SpawnRate=60, Use Spawn Probability=true)
   &SpawnRate(Spawn Probability=$System.SplashE, SpawnGroup=1, SpawnRate=60, Use Spawn Probability=true) {module=SpawnRate001}
   &SpawnRate(Spawn Probability=$System.SplashS, SpawnGroup=2, SpawnRate=60, Use Spawn Probability=true) {module=SpawnRate003}
   &SpawnRate(Spawn Probability=$System.SplashW, SpawnGroup=3, SpawnRate=60, Use Spawn Probability=true) {module=SpawnRate002}
  ]
  particleSpawnScript [
   &InitializeParticle(
    Lifetime Max=0.5
    Lifetime Min=0.4
    Lifetime Mode=::NewEnumerator1
    Mesh Scale Mode=::NewEnumerator3
    Position Offset=<0 0 75>
    Sprite Rotation Mode=::NewEnumerator2
    Sprite Size Mode=::NewEnumerator1
    Uniform Sprite Size Max=45
    Uniform Sprite Size Min=40
    UsePositionOffset=true
    Write Material Random=true
   ) {module=InitializeParticle001}
   &ShapeLocation(
    Disc Coverage=0.3
    MaskSpawnGroup=false
    Offset Mode=::NewEnumerator0
    Override Seed=false
    Ring Radius=27
    Rotation Angle:NiagaraFloat=&Add_Float(A=0.625, B=&Multiply_FloatByInt(Float=0.25, Integer=$Emitter.SpawnGroup))
    Rotation Axis=<0 0 1>
    Rotation Mode=::NewEnumerator1
    Shape Primitive=::NewEnumerator4
    U Distribution=0.75
    UseOwnerScale=true
   )
   &SubUVAnimation(SubUV Animation Mode=::NewEnumerator1)
   &self:SplashVelocityScale() {module=SplashVelocityScale}
   &AddVelocity(
    Module.Use Falloff Distance=false
    Override Seed=false
    Velocity Mode=::NewEnumerator1
    Velocity Speed=&Multiply_Float(A=&RandomRangeFloat(Minimum=100, Maximum=120), B=$Particles.SplashVelocityScale)
   ) {module=AddVelocity001}
  ]
  particleUpdateScript [
   &ParticleState()
   &InheritVelocity(
    Module.Limit Speed=true
    Velocity Limit=10000
    Velocity Scale=&Multiply_VectorByFloat(Vector=<-0.03 -0.03 -0.03>, Float=0.5)
    Velocity Source=&TransformVector(Vector=$Engine.Owner.Velocity, Source Space=::World, Destination Space=::Local)
   )
   &SpriteRotationRate(Rotation Rate=&RandomRangeFloat(Minimum=-90, Maximum=90))
   &Drag(Drag=0.5, Use Linear Drag=true, Use Rotational Drag=false)
   &GravityForce(Gravity=<0 0 -900>)
   &SolveForcesAndVelocity(Clamp Velocity=false, Limit Acceleration=false)
   &ScaleSpriteSize(Scale Factor=&Vector2DFromFloat(Value=&FloatFromCurve(FloatCurve=DataInterface() {propertyValues {Curve {keys [{tangentMode=RCTM_User, time=0, value=0.33}, {tangentMode=RCTM_User, time=1, value=1}]}}})))
   &ScaleColor(
    Scale Alpha=&FloatFromCurve(FloatCurve=DataInterface() {propertyValues {Curve {keys [{interpMode=RCIM_Cubic, tangentMode=RCTM_User, time=0, value=1}, {interpMode=RCIM_Cubic, tangentMode=RCTM_User, time=0.33, value=0.4}]}}})
    Scale RGB=&Lerp_Vector(
     A=<0 0.75 1>
     B=<2 2 2>
     Alpha=&FloatFromCurve(FloatCurve=DataInterface() {propertyValues {Curve {keys [{interpMode=RCIM_Cubic, tangentMode=RCTM_User, time=0, value=0}, {interpMode=RCIM_Cubic, tangentMode=RCTM_User, time=0.45, value=1}]}}})
    )
    ScaleA=false
    ScaleRGB=true
   )
   &DynamicMaterialParameters(
    Index 0 Param 1=&RandomRangeFloat(Minimum=0, Maximum=1)
    Index 0 Param 3=&FloatFromCurve(FloatCurve=DataInterface() {propertyValues {Curve {keys [{interpMode=RCIM_Cubic, tangentMode=RCTM_User, time=0, value=0}, {interpMode=RCIM_Cubic, tangentMode=RCTM_User, time=1, value=1}]}}})
    Index 0 Param 4=&FloatFromCurve(FloatCurve=DataInterface() {propertyValues {Curve {keys [<at 0 0 RCIM_Cubic>, {arriveTangent=1.5276, interpMode=RCIM_Cubic, leaveTangent=1.5276, tangentMode=RCTM_User, time=1, value=1}]}}})
    Module.Param0WriteEnabled=true
    Module.Param1WriteEnabled=false
    Module.Param2WriteEnabled=true
    Module.Param3WriteEnabled=true
   )
  ]
  rendererClasses [&NiagaraSpriteRendererProperties]
  renderers [{rendererClass=&NiagaraSpriteRendererProperties}]
 }
 Splash_Mesh {
  emitterUpdateScript [
   &EmitterState(Life Cycle Mode=::NewEnumerator1, Loop Behavior=::NewEnumerator1, Loop Duration=1, UseLoopDelay=false)
   &SpawnBurst_Instantaneous(Spawn Count=2, Spawn Time=0.1, Use Spawn Probability=false, UseLoopCountLimit=false)
  ]
  particleSpawnScript [
   &InitializeParticle(
    Mesh Scale Mode=::NewEnumerator1
    Mesh Uniform Scale Max=0.09
    Mesh Uniform Scale Min=0.08
    Position Mode=::NewEnumerator0
    Write Material Random=true
   ) {module=InitializeParticle001}
   &InitialMeshOrientation(
    Mesh Orientation Mode=::NewEnumerator1
    Rotation=&RandomRangeVector(Minimum=<0 0 0>, Maximum=<0 0 1>)
    Use Rotation Vector=true
   )
  ]
  particleUpdateScript [
   &ParticleState()
   /SetVariables(Particles.Position=$System.SampleSkeletalMesh.SampledPosition) {isSetParameters=true}
   &DynamicMaterialParameters(
    Index 0 Param 1=&FloatFromCurve(
     FloatCurve=DataInterface() {propertyValues {Curve {keys [{tangentMode=RCTM_User, time=0, value=1}, {tangentMode=RCTM_User, time=1, value=0}]}}}
     Scale Curve=-2
    )
    Index 0 Param 2=&FloatFromCurve(FloatCurve=DataInterface() {propertyValues {Curve {keys [<at -0.0046 4>,<at 1.0011 0>]}}})
    Index 0 Param 3=&FloatFromCurve(FloatCurve=DataInterface() {propertyValues {Curve {keys [
         {arriveTangent=4.1785, interpMode=RCIM_Cubic, leaveTangent=4.1785, tangentMode=RCTM_User, time=0, value=0}
         {arriveTangent=0.9995, interpMode=RCIM_Cubic, leaveTangent=0.9995, tangentMode=RCTM_User, time=0.2223, value=0.6825}
         {arriveTangent=-0.0405, interpMode=RCIM_Cubic, leaveTangent=-0.0405, tangentMode=RCTM_User, time=1, value=1}
        ]}}})
    Index 0 Param 4=&FloatFromCurve(FloatCurve=DataInterface() {propertyValues {Curve {keys [<at 0.2 0 RCIM_Cubic>, {interpMode=RCIM_Cubic, leaveTangent=3.4641, time=1, value=1}]}}})
    Module.Param0WriteEnabled=true
    Module.Param1WriteEnabled=true
    Module.Param2WriteEnabled=true
    Module.Param3WriteEnabled=true
   ) {module=DynamicMaterialParameters001}
   &ScaleColor(
    Scale Alpha=&FloatFromCurve(FloatCurve=DataInterface() {propertyValues {Curve {keys [{interpMode=RCIM_Cubic, tangentMode=RCTM_User, time=0, value=1}, {interpMode=RCIM_Cubic, tangentMode=RCTM_User, time=1, value=0}]}}})
    ScaleA=true
    ScaleRGB=false
   )
   &ScaleMeshSize(Scale Factor=&VectorFromCurve(VectorCurve=DataInterface() {propertyValues {
       XCurve {keys [{arriveTangent=2.6672, interpMode=RCIM_Cubic, leaveTangent=2.6672, tangentMode=RCTM_User, time=0, value=1}, <at 0.9989 2 RCIM_Cubic>]}
       YCurve {keys [{arriveTangent=2.6672, interpMode=RCIM_Cubic, leaveTangent=2.6672, tangentMode=RCTM_User, time=0, value=1}, <at 0.9989 2 RCIM_Cubic>]}
       ZCurve {keys [
         {arriveTangent=1.1096, interpMode=RCIM_Cubic, leaveTangent=1.1096, tangentMode=RCTM_User, time=0, value=1}
         <at 0.5 1.25 RCIM_Cubic>
         {arriveTangent=-1.0344, interpMode=RCIM_Cubic, leaveTangent=-1.0344, tangentMode=RCTM_User, time=0.9989, value=1}
        ]}
      }})) {module=MeshSizeScale}
  ]
  rendererClasses [&NiagaraMeshRendererProperties]
  renderers [{rendererClass=&NiagaraMeshRendererProperties, rendererIndex=0}]
 }
 Water_Drops_Brush {
  emitterUpdateScript [
   &EmitterState()
   &SpawnRate(
    Spawn Probability=0.8
    SpawnRate=&FloatFromCurve(
     FloatCurve=DataInterface() {propertyValues {Curve {keys [
         {arriveTangent=0.2559, time=0, value=0}
         {arriveTangent=1, interpMode=RCIM_Cubic, leaveTangent=1, time=0.8, value=0.4}
         {interpMode=RCIM_Cubic, tangentMode=RCTM_User, time=1, value=1}
        ]}}}
     CurveIndex=$System.Timeline.NormalizedTime
     Scale Curve=30
    )
    Use Spawn Probability=false
   )
  ]
  particleSpawnScript [
   &InitializeParticle(
    Color Minimum=<rgba 1 1 1 0.5>
    Color Mode=::NewEnumerator2
    Lifetime Max=0.75
    Lifetime Min=0.5
    Lifetime Mode=::NewEnumerator1
    Mesh Scale Mode=::NewEnumerator1
    Mesh Uniform Scale Max=3
    Mesh Uniform Scale Min=1.5
    Position Offset=<0 0 72>
    Sprite Size Max=<2 6>
    Sprite Size Min=<1 4>
    Sprite Size Mode=::NewEnumerator2
    UsePositionOffset=true
    Write Material Random=true
   )
   &ShapeLocation(
    Apply To Particle Position=::NewEnumerator1
    MaskSpawnGroup=false
    Non Uniform Scale=<0.75 1 1>
    Override Seed=false
    Ring Radius=7
    Rotation Axis=<0 -1 0>
    Rotation Coordinate Space=::World
    Rotation Mode=::NewEnumerator1
    Shape Origin=&MakeVector(X=0, Y=2, Z=0)
    Shape Primitive=::NewEnumerator4
    U Distribution:NiagaraFloat=&FloatFromCurve(
     FloatCurve=DataInterface() {propertyValues {Curve {keys [{interpMode=RCIM_Cubic, tangentMode=RCTM_User, time=0.5, value=1}, {interpMode=RCIM_Cubic, tangentMode=RCTM_User, time=1, value=0}]}}}
     CurveIndex=$System.Timeline.NormalizedTime
     Scale Curve=0.7
    )
    UseOwnerScale=true
   )
   &SkeletalMeshLocation(
    Bone / Socket Index=6
    Bone Sampling Mode=::NewEnumerator4
    "Inherit Velocity (Scale)"=-0.25
    Inherited Velocity Speed Limit=150
    MaskSpawnGroup=false
    Orientation Sampling=::NewEnumerator1
    Sampled Position Offset=$Particles.Initial.ShapeLocation.ShapePosition
    Skeletal Mesh=$User.SkeletalMesh
    Velocity Sampling=::NewEnumerator0
    VelocitySpeedLimit=false
   ) {module=SkeletalMeshLocation001}
   &DynamicMaterialParameters(
    Index 0 Param 1=1
    Module.Param0WriteEnabled=true
    Module.Param1WriteEnabled=false
    Module.Param2WriteEnabled=false
    Module.Param3WriteEnabled=false
   )
   &ApplyInitialForces(Apply Rotational Force to Rotational Velocity=false) {enabled=false}
  ]
  particleUpdateScript [
   &ParticleState()
   &Drag(Drag=0.1, Use Linear Drag=true, Use Rotational Drag=false)
   &GravityForce(Gravity=<0 0 -600>)
   &SolveForcesAndVelocity(Clamp Velocity=false, Limit Acceleration=false)
   &ScaleSpriteSize(Uniform Curve Sprite Scale=DataInterface() {propertyValues {Curve {keys [<at 0 0 RCIM_Cubic>, {tangentMode=RCTM_User, time=0.1, value=1}]}}})
   &ScaleColor(
    Scale Alpha=&FloatFromCurve(FloatCurve=DataInterface() {propertyValues {Curve {keys [
         {arriveTangent=28.4939, interpMode=RCIM_Cubic, leaveTangent=28.4939, tangentMode=RCTM_User, time=0, value=0}
         {arriveTangent=0.002, interpMode=RCIM_Cubic, leaveTangent=0.002, time=0.1183, value=1.0113}
         {arriveTangent=-3.6864, interpMode=RCIM_Cubic, leaveTangent=-3.6864, tangentMode=RCTM_User, time=1.0011, value=0.002}
        ]}}})
    ScaleA=true
    ScaleRGB=true
   )
  ]
  rendererClasses [&NiagaraSpriteRendererProperties]
  renderers [{rendererClass=&NiagaraSpriteRendererProperties}]
 }
}
summary {
 systemName=NS_Emote_ScrubADub_Tub
 emitters [
  {emitterName=Bubbles_Flowing, rendererClasses [&NiagaraMeshRendererProperties]}
  {
   emitterName=FN_Water_Drops_
   rendererClasses [&NiagaraSpriteRendererProperties, &NiagaraMeshRendererProperties]
  }
  {emitterName=Splash_Mesh, rendererClasses [&NiagaraMeshRendererProperties]}
  {emitterName=Splash_Foam, rendererClasses [&NiagaraSpriteRendererProperties]}
  {emitterName=Splash_Foam_Dynamic, rendererClasses [&NiagaraSpriteRendererProperties]}
  {emitterName=FN_Water_Drops_001, rendererClasses [&NiagaraSpriteRendererProperties]}
  {emitterName=Water_Drops_Brush, rendererClasses [&NiagaraSpriteRendererProperties]}
  {emitterName=Foam_Brush, rendererClasses [&NiagaraSpriteRendererProperties]}
  {emitterName=Bubbles_Brush, rendererClasses [&NiagaraMeshRendererProperties]}
 ]
 userVariables [
  User.FoamSpawn {type=NiagaraFloat, defaultValue=0, description=""}
  User.SkeletalMesh {
   type=NiagaraDataInterfaceSkeletalMesh
   defaultValue {
    struct=NiagaraExt_VariableValue_DataInterface
    value {dataInterface=&self:NiagaraDataInterfaceSkeletalMesh_0}
   }
  }
 ]
}
